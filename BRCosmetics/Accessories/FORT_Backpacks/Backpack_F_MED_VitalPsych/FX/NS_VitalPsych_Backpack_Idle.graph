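# Niagara system for the idle effect on the VitalPsych backpack: a persistent logo-like core with a flickering pink glow.
# Core burst-spawns a sprite and mesh rendered particle that follows System.LogoOffset, with sprite alignment rotated by System.AutoRotate.
# PresetFX_Glow tracks the same offset with sprite and light renderers, pulsing its color through a curve on looped system age.
# User material and texture parameters default to EerieGhost backpack hologram assets.
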NiagaraSystem /BRCosmetics/Accessories/FORT_Backpacks/Backpack_F_MED_VitalPsych/FX/NS_VitalPsych_Backpack_Idle
dependencies {
 usedDataInterfaces [&NiagaraDataInterfaceCurve]
 usedDynamicInputs [
  &RandomRangeFloat
  &MultiplyQuaternion
  &MakeQuatFromAxisAngle
  &Lerp_Vector
  &Sine
  &Add_Float
  &Lerp_Float
  &MakeVector
  &Subtract_Float
 ]
 usedModules [
  &self:SystemSpawnScript.SystemScriptSource.SystemScriptGraph.NiagaraNodeAssignment_0.SetVariables_#1
  &SystemState
  &self:SystemSpawnScript.SystemScriptSource.SystemScriptGraph.NiagaraNodeAssignment_1.SetVariables_#2
  &self:SystemSpawnScript.SystemScriptSource.SystemScriptGraph.NiagaraNodeAssignment_2.SetVariables_#3
 ]
}
emitters {
 Core {
  emitterUpdateScript [
   &EmitterState(
    Inactive Response=::NewEnumerator1
    Life Cycle Mode=::NewEnumerator1
    Loop Behavior=::NewEnumerator1
    Loop Duration Mode=::NewEnumerator1
    UseLoopDelay=false
   )
   &SpawnBurst_Instantaneous(Use Spawn Probability=false, UseLoopCountLimit=false)
  ]
  particleSpawnScript [
   &InitializeParticle(
    Lifetime=5
    Mesh Scale Mode=::NewEnumerator3
    Sprite Size Mode=::NewEnumerator3
    Uniform Sprite Size=35
    UsePositionOffset=true
    Write Material Random=true
   )
   &DynamicMaterialParameters(
    Index 0 Param 1=$System.Random
    Index 0 Param 2=&MakeFloatFromVector(Vector=$Particles.Scale, Channel=::NewEnumerator0)
    Module.Param0WriteEnabled=true
    Module.Param1WriteEnabled=true
    Module.Param2WriteEnabled=false
    Module.Param3WriteEnabled=false
   ) {enabled=false}
  ]
  particleUpdateScript [
   &SpriteFacingAndAlignment(
    Sprite Alignment=&RotateVectorByQuaternion(Vector To Rotate=<0 0.2 1>, Quaternion=$System.AutoRotate)
    Sprite Facing=<0 -1 0>
    WriteAlignment=true
    WriteFacing=true
   )
   &ParticleState(Kill Particles When Lifetime Has Elapsed=false, Let Infinitely Lived Particles Die When Emitter Deactivates=true)
   /SetVariables(Particles.Position=&Add_Vector(A=<0 0 0>, B=$System.LogoOffset)) {isSetParameters=true}
   &ScaleColor(
    Scale RGB=&MakeVectorFromLinearColorRGB(Color=&FN_Select_Color_from_Time_Of_Day(Morning=<rgba 1 1 1 1>, Day=<rgba 1 1 1 1>, Evening=<rgba 1 1 1 1>, Night=<rgba 1 1 1 1>))
    ScaleA=true
    ScaleRGB=true
   ) {enabled=false}
  ]
  rendererClasses [&NiagaraSpriteRendererProperties, &NiagaraMeshRendererProperties]
  renderers [
   {rendererClass=&NiagaraSpriteRendererProperties}
   {rendererClass=&NiagaraMeshRendererProperties, rendererIndex=1}
   {rendererClass=&NiagaraMeshRendererProperties, rendererIndex=2}
  ]
 }
 PresetFX_Glow {
  emitterUpdateScript [
   &EmitterState(
    Inactive Response=::NewEnumerator1
    Life Cycle Mode=::NewEnumerator1
    Loop Behavior=::NewEnumerator1
    Loop Duration Mode=::NewEnumerator1
    UseLoopDelay=false
   )
   &SpawnBurst_Instantaneous(Use Spawn Probability=false, UseLoopCountLimit=false)
  ]
  particleSpawnScript [&InitializeParticle(
    Color=<rgba 4 1.0878 3.8099 0.02>
    Sprite Size Mode=::NewEnumerator3
    Uniform Sprite Size=50
    UsePositionOffset=false
    Write Material Random=true
   )]
  particleUpdateScript [
   &ParticleState(Kill Particles When Lifetime Has Elapsed=false)
   &ScaleColor(
    Scale RGB=&VectorFromFloat(Value=&FloatFromCurve(
      FloatCurve=DataInterface() {propertyValues {Curve {keys [<at 0 1>,<at 0.1 0.625>,<at 0.2 1.125>,<at 0.4 0.375>,<at 0.5 1.375>,<at 0.6 0.75>,<at 0.7 1.625>,<at 0.8 0.625>,<at 0.9 1.25>,<at 1 0.875>]}}}
      CurveIndex=&Multiply_Float(A=$System.LoopedAge, B=0.2)
     ))
    ScaleA=true
    ScaleRGB=true
   )
   /SetVariables(Particles.Position=&Add_Vector(A=<0 0 -5>, B=$System.LogoOffset)) {isSetParameters=true}
   &CameraOffset(Camera Offset Amount=10)
  ]
  rendererClasses [&NiagaraSpriteRendererProperties, &NiagaraLightRendererProperties]
  renderers [
   {rendererClass=&NiagaraSpriteRendererProperties}
   {rendererClass=&NiagaraLightRendererProperties, rendererIndex=1}
  ]
 }
}
summary {
 systemName=NS_VitalPsych_Backpack_Idle
 emitters [
  {emitterName=Core, rendererClasses [&NiagaraSpriteRendererProperties, &NiagaraMeshRendererProperties]}
  {emitterName=PresetFX_Glow, rendererClasses [&NiagaraSpriteRendererProperties, &NiagaraLightRendererProperties]}
 ]
 userVariables [
  User.Backpack_DiffuseTex {
   type=Texture
   description=""
   defaultValue {
    struct=NiagaraExt_VariableValue_Object
    value {object=/BRCosmetics/Accessories/FORT_Backpacks/Backpack_M_MED_EerieGhost/FX/T_FX_EerieGhost_HologramSrc, objectClass=/Script/Engine.Texture}
   }
  }
  User.Backpack_Mat {
   type=MaterialInterface
   description=""
   defaultValue {
    struct=NiagaraExt_VariableValue_Object
    value {
     object=/BRCosmetics/Accessories/FORT_Backpacks/Backpack_M_MED_EerieGhost/FX/Material/MI_3DHB_EerieGhost
     objectClass=/Script/Engine.MaterialInterface
    }
   }
  }
  User.Backpack_NormalTex {
   type=Texture
   description=""
   defaultValue {
    struct=NiagaraExt_VariableValue_Object
    value {object=/BRCosmetics/Accessories/FORT_Backpacks/Backpack_M_MED_EerieGhost/FX/Texture/T__Normal, objectClass=&Texture}
   }
  }
  User.Backpack_WPOTex {
   type=Texture
   description=""
   defaultValue {
    struct=NiagaraExt_VariableValue_Object
    value {object=/BRCosmetics/Accessories/FORT_Backpacks/Backpack_M_MED_EerieGhost/FX/Texture/T__WPO, objectClass=&Texture}
   }
  }
  User.IsFrom3D {type=NiagaraFloat, defaultValue=0, description=""}
 ]
}
